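; Massage table mirror prop that broadcasts a LookedAtMirror quest verb when a player in the volume in front of it faces the mirror.
; Volume begin overlap casts to FortPlayerPawn, stores the actor and starts a looping timer, which end overlap clears.
; Each timer tick compares the pawn's yaw with the look-at yaw to the static mesh within a tolerance, ignoring pitch.
; Only one actor and timer handle are stored.
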(event EventBeginPlay)

(event Collision|EventActorBeginOverlap
  (params (OtherActor :type Actor%20Object%20Reference)))

(event EventTick
  (params (DeltaSeconds :type Float%20%28single-precision%29)))

(event OnComponentBeginOverlap(IsPlayerLookingVolume)
  (params (OverlappedComponent :type Primitive%20Component%20Object%20Reference) (OtherActor :type Actor%20Object%20Reference) (OtherComp :type Primitive%20Component%20Object%20Reference) (OtherBodyIndex :type Integer) (bFromSweep :type Boolean) (SweepResult :type Hit%20Result%20Structure%20%28by%20ref%29))
  ;[0 0 0] Quest - send event if player is looking into the mirror when they are in the volume in front of it
  (Utilities|IsValid OtherActor
    (:Is%20Valid
      (Utilities|Casting|CastToFortPlayerPawn OtherActor)
      (Variables|QuestLookInMirror|SetOther%20Actor OtherActor)
      (bind _returnvalue (Utilities|Time|SetTimerbyEvent (delegate Custom|Tick) 1.0 true true 0.5))
      (Variables|QuestLookinMirror|SetTimer%20Handle _returnvalue))))

(event Custom|Tick
  ;[0 0 0] Quest - send event if player is looking into the mirror when they are in the volume in front of it
  (Utilities|IsValid (Variables|QuestLookInMirror|GetOther%20Actor)
    (:Is%20Valid
      (if (Math|Float|NearlyEqual(Float) (.yaw (Transformation|GetActorRotation (Game|Player|GetPlayerPawn (Fortnite|GetFortPlayerControllerfromActor (Variables|QuestLookInMirror|GetOther%20Actor))))) (.yaw (Math|Rotator|FindLookatRotation (Transformation|GetActorLocation (Game|Player|GetPlayerPawn (Fortnite|GetFortPlayerControllerfromActor (Variables|QuestLookInMirror|GetOther%20Actor)))) (Transformation|GetWorldLocation (Variables|Debug|GetStaticMeshComponent)))) 15.0)
        (CustomVerb|BroadcastCustomVerbMessage (CustomVerb|MakeCustomVerbMessage (Variables|QuestLookInMirror|GetOther%20Actor) "(TagName=\"CustomVerb.LookedAtMirror\")"))))))

(event OnComponentEndOverlap(IsPlayerLookingVolume)
  (params (OverlappedComponent :type Primitive%20Component%20Object%20Reference) (OtherActor :type Actor%20Object%20Reference) (OtherComp :type Primitive%20Component%20Object%20Reference) (OtherBodyIndex :type Integer))
  (bind _timer_handle (Variables|QuestLookinMirror|GetTimer%20Handle))
  ;[0 0 0] Quest - send event if player is looking into the mirror when they are in the volume in front of it
  (if (Utilities|Time|IsValidTimerHandle _timer_handle)
    (Utilities|Time|ClearandInvalidateTimerbyHandle _timer_handle)))

(fn ConstructionScript
  (|CallParentFunction))
